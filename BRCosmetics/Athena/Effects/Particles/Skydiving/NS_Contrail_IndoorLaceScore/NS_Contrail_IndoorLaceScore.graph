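# Niagara skydiving contrail cosmetic that trails ribbons, rotating clock sprites, numbers, stars and dust from two skeletal bones or sockets.
# FallingStars seeds StarsDustTrail through particle reads, while Minimal packs stars, dust and numbers into one emitter using spawn groups and per-renderer attributes.
# VariantPicker switches between purple and orange palettes, and ParaGlideLeanAlpha adjusts spawn, size and drift.
# Front-end, time-of-day and platform modifiers tune opacity, offsets and brightness.
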NiagaraSystem /BRCosmetics/Athena/Effects/Particles/Skydiving/NS_Contrail_IndoorLaceScore/NS_Contrail_IndoorLaceScore
dependencies {
 usedDataInterfaces [
  &NiagaraDataInterfaceArrayFloat
  &NiagaraDataInterfaceVectorCurve
  &NiagaraDataInterfaceColorCurve
  &NiagaraDataInterfacePlatformSet
  &NiagaraDataInterfaceCurve
  &NiagaraDataInterfaceParticleRead
 ]
 usedDynamicInputs [&SimulationPosition]
 usedModules [
  &self:SystemSpawnScript.SystemScriptSource.SystemScriptGraph.NiagaraNodeAssignment_1.SetVariables_#1
  &self:SystemSpawnScript.SystemScriptSource.SystemScriptGraph.NiagaraNodeAssignment_3.SetVariables_#2
  &SystemState
  /Game/Effects/Niagara/Modules/Skydiving/FN_DivingAttributes/v5/FN_DivingAttributes
  &self:SystemSpawnScript.SystemScriptSource.SystemScriptGraph.NiagaraNodeAssignment_0.SetVariables_#3
  &self:SystemSpawnScript.SystemScriptSource.SystemScriptGraph.NiagaraNodeAssignment_4.SetVariables_#4
 ]
}
emitters {
 BigRibbon {
  emitterUpdateScript [
   &EmitterState()
   &FN_SpawnAnimTrail(
    Number of Anim Trails=2
    Spawn Mode:ENiagaraSkeletalSpawning=::NewEnumerator0
    Switch on ENiagara_AnimTrail_Source:ENiagara_AnimTrail_Source=::NewEnumerator1
   )
  ]
  particleSpawnScript [
   &FN_InitializeAnimTrailBoneIndecies(Second Bone or Socket Index=0)
   &FN_InitializeAnimTrailBoneIndecies(AnimTrail Selection=1, First Bone or Socket Index=1) {module=FN_InitializeAnimTrailBoneIndecies001}
   &FN_InitializeAnimTrail(AnimTrailWidth=90, Lifetime=0.35, SkeletalMesh=$System.Skeletal Mesh Reader)
   &AddVelocity(
    Scale Added Velocity=&Lerp_Vector(A=<1 0 1>, B=<1 1 2>, Alpha=$System.FN_Integrator001.Value)
    Velocity=<0 -100 500>
   )
  ]
  particleUpdateScript [
   &FN_UpdateAnimTrail(SkeletalMesh=$System.Skeletal Mesh Reader)
   &SolveForcesAndVelocity(Clamp Velocity=false, Limit Acceleration=false)
   &Color(
    Color=&MergeHelper_ColorSwitch(
     0=&ColorFromCurve(ColorCurve=DataInterface() {propertyValues {
        AlphaCurve {keys [<at 0 1>]}
        BlueCurve {keys [<at 0 1>, <at 0.2967 1>, {arriveTangent=-0.7924, interpMode=RCIM_Cubic, leaveTangent=-0.7924, time=0.6709, value=1}, <at 1 0.4427>]}
        GreenCurve {keys [<at 0 0.5469>, <at 0.2967 0.2448>, {arriveTangent=-0.3262, interpMode=RCIM_Cubic, leaveTangent=-0.3262, time=0.6709, value=0}, <at 1 0.0154>]}
        RedCurve {keys [<at 0 0.8806>, <at 0.2967 0.5731>, {arriveTangent=-0.8149, interpMode=RCIM_Cubic, leaveTangent=-0.8149, time=0.6709, value=0.1739}, <at 1 0>]}
       }})
     1=&ColorFromCurve(ColorCurve=DataInterface() {propertyValues {
        AlphaCurve {keys [<at 0 1>]}
        BlueCurve {keys [<at 0 0>, <at 0.1935 0.0359>, {arriveTangent=-0.1293, interpMode=RCIM_Cubic, leaveTangent=-0.1293, time=0.2445, value=0}, <at 0.4709 0>]}
        GreenCurve {keys [
          <at 0 0.5462>
          <at 0.1935 0.6179>
          {arriveTangent=-0.7253, interpMode=RCIM_Cubic, leaveTangent=-0.7253, time=0.2445, value=0.5833}
          <at 0.4709 0.4167>
         ]}
        RedCurve {keys [<at 0 1>,<at 0.1935 1>,<at 0.2445 1 RCIM_Cubic>,<at 0.4709 1>]}
       }})
     2=?
     3=?
     Index=&MakeIntFromFloat(Float=$User.VariantPicker)
     Num=::NewEnumerator0
    )
    Scale Alpha=&FloatFromCurve(
     FloatCurve=DataInterface() {propertyValues {Curve {keys [
         {arriveTangent=0.7543, interpMode=RCIM_Cubic, leaveTangent=0.7543, tangentMode=RCTM_User, time=0, value=0}
         <at 0.1 0.05>
         {arriveTangent=-0.013, interpMode=RCIM_Cubic, leaveTangent=-0.013, tangentMode=RCTM_User, time=0.2992, value=0.0096}
         {interpMode=RCIM_Cubic, leaveTangent=-0.2611, time=1, value=0}
        ]}}}
     Scale Curve=&Multiply_Float(
      A=&FN_FrontEndFloatModifier(
       Gameplay=5
       Front End=&MergeHelper_FloatSwitch(0=2, 1=1, 2=?, 3=?, Index=&MakeIntFromFloat(Float=$User.VariantPicker), Num=::NewEnumerator0)
      )
      B=$System.FadeIn
     )
    )
   )
   &Material(Base Texture Time:NiagaraFloat=&Multiply_Float(A=$Emitter.Age, B=-1), PostMult=0.5455, Power=0.4408)
  ]
  rendererClasses [&NiagaraRibbonRendererProperties]
  renderers [{rendererClass=&NiagaraRibbonRendererProperties}]
 }
 ClockMain {
  emitterUpdateScript [
   &EmitterState(
    Inactive Response=::NewEnumerator1
    Life Cycle Mode=::NewEnumerator1
    Loop Behavior=::NewEnumerator1
    Loop Duration Mode=::NewEnumerator1
    UseLoopDelay=true
   )
   &SpawnBurst_Instantaneous(Spawn Count=2, Use Spawn Probability=false)
  ]
  particleSpawnScript [
   &FN_SetSpawnGroup(Number of Spawn Groups=2)
   &InitializeParticle(
    Color=&MergeHelper_ColorSwitch(
     0=<rgba 0.9043 0 4 1>
     1=<rgba 4 1.1067 0 1>
     2=?
     3=?
     Index=&MakeIntFromFloat(Float=$User.VariantPicker)
     Num=::NewEnumerator0
    )
    Lifetime=2
    Sprite Size Mode=::NewEnumerator1
    Uniform Sprite Size Max=40
    Uniform Sprite Size Min=38
    UsePositionOffset=false
    Write Material Random=true
   ) {module=InitializeParticle001}
  ]
  particleUpdateScript [
   &ParticleState(Kill Particles When Lifetime Has Elapsed=false, Let Infinitely Lived Particles Die When Emitter Deactivates=true)
   &ScaleSpriteSize(Uniform Curve Sprite Scale=DataInterface() {propertyValues {Curve {
       defaultValue=0
       keys [
        {arriveTangent=7.359, interpMode=RCIM_Cubic, leaveTangent=7.359, tangentMode=RCTM_User, time=0, value=0}
        {interpMode=RCIM_Cubic, tangentMode=RCTM_User, time=0.2664, value=1}
       ]
      }}})
   &SpriteRotationRate(Rotation Rate=&Multiply_Float(
     A=&UniformAOrBFloat(
      A=&Lerp_Float(A=0.3, B=1, Alpha=$User.ParaGlideLeanAlpha)
      B=&Lerp_Float(A=0.3, B=1, Alpha=$User.ParaGlideLeanAlpha)
      Evaluation Type=::NewEnumerator0
      Recalculate Random Each Loop=true
      Override Seed=false
     )
     B=&RandomRangeFloat(Minimum=100, Maximum=250)
    )) {module=SpriteRotationRate001}
   &SkeletalMeshLocation(
    Bone / Socket Index=$Particles.SpawnGroup
    Bone Sampling Mode=::NewEnumerator4
    MaskSpawnGroup=false
    Skeletal Mesh=$System.Skeletal Mesh Reader
    Write To Particle Position=false
   )
   /Game/Effects/Niagara/Modules/Skydiving/FN_SkydivePositionOffset_v1/FN_SkydivePositionOffset(
    Bone Orientation:Quat4f=$Particles.SkeletalMeshLocation.SampledRotation
    Bone Position:NiagaraPosition=$Particles.SkeletalMeshLocation.SampledPosition
    Offset:NiagaraFloat=/Game/Effects/Niagara/DynamicInputs/Skydiving/SkydiveHandsFeetFloatModifier(Hands:NiagaraFloat=0, Feet:NiagaraFloat=0)
   )
   &DynamicMaterialParameters(
    Index 0 Param 1=2
    Index 0 Param 2=1
    Index 0 Param 3=1
    Index 0 Param 4=1
    Module.Index0Float0Write=true
    Module.Index0Float1Write=true
    Module.Index0Float2Write=true
    Module.Index0Float3Write=true
   )
   &ScaleColor(
    Scale Alpha=&FNTimeOfDayScalar(Night Scale=0.8, Morning Scale=1, Evening Scale=1)
    ScaleA=true
    ScaleRGB=true
   ) {module=ScaleColor002}
   /SetVariables(
    Particles.LongShadowPosition:NiagaraPosition=&AddVectorToPosition(
     Position=$Particles.SkeletalMeshLocation.SampledPosition
     Vector=&FN_FrontEndVector3Modifier(
      Gameplay=&Lerp_Vector(A=<0 0 35>, B=<0 0 40>, Alpha=$User.ParaGlideLeanAlpha)
      Front End=&Lerp_Vector(A=<0 0 35>, B=<0 -10 40>, Alpha=$User.ParaGlideLeanAlpha)
     )
    )
    Particles.LongSpriteSize:Vector2f=&Lerp_Vector2(A=<90 190>, B=<90 250>, Alpha=$User.ParaGlideLeanAlpha)
    Particles.ShadowColor:LinearColor=&MakeLinearColorFromVectorAndFloat(
     "Vector (RGB)"=&MakeVectorFromLinearColorRGB(Color=&MergeHelper_ColorSwitch(
       0=<rgba 0.0616 0 0.3542 5>
       1=<rgba 0.1823 0.0414 0.0175 1>
       2=?
       3=?
       Index=&MakeIntFromFloat(Float=$User.VariantPicker)
       Num=::NewEnumerator0
      ))
     "Float (Alpha)"=&Multiply_Float(
      A=&Multiply_Float(
       A=&Multiply_Float(A=$System.FadeIn, B=&ClampFloat(Float=0.8))
       B=&MergeHelper_FloatSwitch(
        0=&FNTimeOfDayScalar(Night Scale=1, Morning Scale=0.6, Evening Scale=0.8)
        1=&FNTimeOfDayScalar(Night Scale=0.4, Morning Scale=1, Evening Scale=1)
        2=?
        3=?
        Index=&MakeIntFromFloat(Float=$User.VariantPicker)
        Num=::NewEnumerator0
       )
      )
      B=&FN_FrontEndFloatModifier(
       Gameplay=1
       Front End=&MergeHelper_FloatSwitch(0=1, 1=0.6, 2=?, 3=?, Index=&MakeIntFromFloat(Float=$User.VariantPicker), Num=::NewEnumerator0)
      )
     )
    )
    Particles.ShadowSize:Vector2f=&Vector2DFromFloat(Value=90)
   ) {isSetParameters=true}
   &SpriteFacingAndAlignment(
    Sprite Alignment=&FN_FrontEndVector3Modifier(Gameplay=<0 0 1>, Front End=&Lerp_Vector(A=<0 0 1>, B=<0 -0.4 1>, Alpha=$User.ParaGlideLeanAlpha))
    WriteAlignment=true
    WriteFacing=false
   )
   &FN_Light_Attributes(
    Color=<rgba 1 0 0.9051 1>
    Create Color Attribute=true
    Create Exponent Attribute=true
    Create Position Attribute=true
    Create Radius Attribute=true
    Position:NiagaraPosition=$Output.SkeletalMeshLocation.SampledPositionWithoutOffset
    Radius=20
   )
   /SetVariables() {isSetParameters=true}
  ]
  rendererClasses [&NiagaraSpriteRendererProperties, &NiagaraLightRendererProperties]
  renderers [
   {rendererClass=&NiagaraSpriteRendererProperties}
   {rendererClass=&NiagaraSpriteRendererProperties, rendererIndex=1}
   {rendererClass=&NiagaraSpriteRendererProperties, rendererIndex=2}
   {rendererClass=&NiagaraLightRendererProperties, rendererIndex=3}
  ]
 }
 ClockSecondary {
  emitterUpdateScript [
   &EmitterState(Inactive Response=::NewEnumerator1, Life Cycle Mode=::NewEnumerator1, Loop Duration=1, UseLoopDelay=true)
   &SpawnRate(SpawnRate=&Lerp_Float(A=2, B=6, Alpha=$User.ParaGlideLeanAlpha), Use Spawn Probability=false)
   &SpawnBurst_Instantaneous(Spawn Count=4, Use Spawn Probability=false)
  ]
  particleSpawnScript [
   &FN_SetSpawnGroup(Number of Spawn Groups=2)
   &InitializeParticle(
    Lifetime=&Lerp_Float(A=2, B=1.5, Alpha=$User.ParaGlideLeanAlpha)
    Sprite Rotation Angle=&RandomRangeFloat(Minimum=0, Maximum=1)
    Sprite Rotation Mode=::NewEnumerator1
    Sprite Size Mode=::NewEnumerator1
    Uniform Sprite Size Max=70
    Uniform Sprite Size Min=50
    UsePositionOffset=false
    Write Material Random=true
   ) {module=InitializeParticle001}
   &DynamicMaterialParameters(Module.Index0Float0Write=true, Module.Index0Float1Write=true, Module.Index0Float2Write=true, Module.Index0Float3Write=true) {module=DynamicMaterialParameters001}
  ]
  particleUpdateScript [
   &ParticleState()
   &SkeletalMeshLocation(
    Bone / Socket Index=$Particles.SpawnGroup
    Bone Sampling Mode=::NewEnumerator4
    MaskSpawnGroup=false
    Skeletal Mesh=$System.Skeletal Mesh Reader
    Write To Particle Position=false
   )
   &ScaleSpriteSize(
    Scale Sprite Size Mode=::NewEnumerator0
    Uniform Scale Factor=&Lerp_Float(
     A=&FloatFromCurve(FloatCurve=DataInterface() {propertyValues {Curve {
         defaultValue=0
         keys [
          {arriveTangent=7.359, interpMode=RCIM_Cubic, leaveTangent=7.359, tangentMode=RCTM_User, time=0, value=0}
          {interpMode=RCIM_Cubic, tangentMode=RCTM_User, time=0.2664, value=1}
         ]
        }}})
     B=&FloatFromCurve(
      FloatCurve=DataInterface() {propertyValues {Curve {
         defaultValue=0
         keys [
          {arriveTangent=2.7121, interpMode=RCIM_Cubic, leaveTangent=2.7121, tangentMode=RCTM_User, time=0, value=0}
          {interpMode=RCIM_Cubic, tangentMode=RCTM_User, time=0.6488, value=1}
         ]
        }}}
      Scale Curve=1.5
     )
     Alpha=$User.ParaGlideLeanAlpha
    )
   ) {module=ScaleSpriteSize001}
   &ScaleSpriteSize(Uniform Curve Sprite Scale=DataInterface() {propertyValues {Curve {
       defaultValue=0
       keys [
        {arriveTangent=7.359, interpMode=RCIM_Cubic, leaveTangent=7.359, tangentMode=RCTM_User, time=0, value=0}
        {interpMode=RCIM_Cubic, tangentMode=RCTM_User, time=0.2664, value=1}
       ]
      }}})
   &Color(Color=&MakeLinearColorFromVectorAndFloat(
     "Vector (RGB)"=&MakeVectorFromLinearColorRGB(Color=&MergeHelper_ColorSwitch(
       0=<rgba 0.778 0.1198 1 1>
       1=<rgba 1 0.4872 0.1198 1>
       2=?
       3=?
       Index=&MakeIntFromFloat(Float=$User.VariantPicker)
       Num=::NewEnumerator0
      ))
     "Float (Alpha)"=&Lerp_Float(
      A=&FloatFromCurve(
       FloatCurve=DataInterface() {propertyValues {Curve {keys [{arriveTangent=0.0516, interpMode=RCIM_Cubic, time=-0.003, value=0}, <at 0.2978 0.2468 RCIM_Cubic>, <at 1 0>]}}}
       Scale Curve=0.5
      )
      B=&FloatFromCurve(
       FloatCurve=DataInterface() {propertyValues {Curve {keys [
           {arriveTangent=0.0516, interpMode=RCIM_Cubic, time=-0.003, value=0}
           {arriveTangent=-0.7399, interpMode=RCIM_Cubic, leaveTangent=-0.7399, tangentMode=RCTM_User, time=0.2003, value=0.2468}
           <at 1 0>
          ]}}}
       Scale Curve=0.5
      )
      Alpha=$User.ParaGlideLeanAlpha
     )
    ))
   &FN_SkydivePositionOffset(
    Bone Orientation:Quat4f=$Particles.SkeletalMeshLocation.SampledRotation
    Bone Position:NiagaraPosition=$Particles.SkeletalMeshLocation.SampledPosition
    Offset:NiagaraFloat=&SkydiveHandsFeetFloatModifier(Hands:NiagaraFloat=0, Feet:NiagaraFloat=0)
   )
   &DynamicMaterialParameters(
    Index 0 Param 1=1
    Index 0 Param 3=0.1
    Index 0 Param 4=&FloatFromCurve(FloatCurve=DataInterface() {propertyValues {Curve {keys [
         {arriveTangent=-5.637, interpMode=RCIM_Cubic, time=0, value=0.9968}
         {arriveTangent=-0.9983, interpMode=RCIM_Cubic, leaveTangent=-0.9983, time=0.3196, value=0.0047}
         <at 0.9984 0 RCIM_Cubic>
        ]}}})
    Module.Param0WriteEnabled=true
    Module.Param1WriteEnabled=false
    Module.Param2WriteEnabled=true
    Module.Param3WriteEnabled=true
   )
   /SetVariables(
    Particles.GlitterColor:LinearColor=&MakeLinearColorFromVectorAndFloat(
     "Vector (RGB)"=&VectorFromCurve(
      VectorCurve=DataInterface() {propertyValues {XCurve {keys [<at 0 1>,<at 1 0>]}, YCurve {keys [<at 0 1>,<at 1 0>]}, ZCurve {keys [<at 0 1>,<at 1 0>]}}}
      Scale Curve=&Multiply_VectorByFloat(
       Vector=<0.1 0.1 0.1>
       Float=&FN_PickFloatDependOnPlatform(TargetPlatformActive=1, TargetPlatformNoActive=0.4)
      )
     )
     "Float (Alpha)"=&ClampFloat(
      Float=&FloatFromCurve(FloatCurve=DataInterface() {propertyValues {Curve {keys [{arriveTangent=-1, time=0, value=1}, <at 1 0>]}}})
      Min=0.5
     )
    )
    Particles.GlitterDynamicMaterialParameter:Vector4f=&MakeVector4_FromLinearColor(LinearColor=&ColorFromCurve(
      ColorCurve=DataInterface() {propertyValues {
        AlphaCurve {keys [<at 0 1 RCIM_Cubic>,<at 1 1 RCIM_Cubic>]}
        BlueCurve {keys [<at 0 0>,<at 0.4992 0>,<at 1 0.1066 RCIM_Cubic>]}
        GreenCurve {keys [<at 0 0>,<at 0.4992 1>,<at 1 0 RCIM_Cubic>]}
        RedCurve {keys [<at 0 1>,<at 0.4992 0>,<at 1 1 RCIM_Cubic>]}
       }}
      CurveIndex=HlslExpression() {hlslExpression="Modulo(Particles.NormalizedAge*2 + Particles.MaterialRandom, 1.0)"}
     ))
    Particles.GlitterSpriteSize:Vector2f=&Vector2DFromFloat(Value=&FloatFromCurve(FloatCurve=DataInterface() {propertyValues {Curve {keys [<at 0 1>,<at 1 5>]}}}, Scale Curve=14))
    Particles.GlitterSubImageIndex:NiagaraFloat=&FloatFromCurve(
     FloatCurve=DataInterface() {propertyValues {Curve {defaultValue=0, keys [{tangentMode=RCTM_User, time=0, value=0}, {tangentMode=RCTM_User, time=1, value=1}]}}}
     Scale Curve=49
    )
    Particles.RingColor:LinearColor=&MakeLinearColorFromVectorAndFloat(
     "Vector (RGB)"=&MakeVectorFromLinearColorRGB(Color=<rgba 3 1.693 2.9594 1>)
     "Float (Alpha)"=&FloatFromCurve(
      FloatCurve=DataInterface() {propertyValues {Curve {
         defaultValue=1
         keys [
          {interpMode=RCIM_Cubic, tangentMode=RCTM_User, time=0, value=1}
          {arriveTangent=-0.0386, interpMode=RCIM_Cubic, leaveTangent=-0.0386, tangentMode=RCTM_User, time=0.9945, value=0}
         ]
        }}}
      Scale Curve=&Lerp_Float(A=0, B=1, Alpha=$User.ParaGlideLeanAlpha)
     )
    )
    Particles.RingDynamicMaterialParameter:Vector4f=&MakeVector4(
     X=1
     Y=0
     Z=0
     W=&FloatFromCurve(FloatCurve=DataInterface() {propertyValues {Curve {
         defaultValue=1
         keys [{interpMode=RCIM_Cubic, tangentMode=RCTM_User, time=0, value=0}, {interpMode=RCIM_Cubic, tangentMode=RCTM_User, time=1, value=1}]
        }}})
    )
    Particles.RingSpriteSize:Vector2f=&Vector2DFromFloat(Value=&Multiply_Float(
      A=&FloatFromCurve(FloatCurve=DataInterface() {propertyValues {Curve {
          defaultValue=0
          keys [{interpMode=RCIM_Cubic, tangentMode=RCTM_User, time=0, value=0}, {interpMode=RCIM_Cubic, tangentMode=RCTM_User, time=1, value=1}]
         }}})
      B=60
     ))
   ) {isSetParameters=true}
  ]
  rendererClasses [&NiagaraSpriteRendererProperties]
  renderers [
   {rendererClass=&NiagaraSpriteRendererProperties}
   {rendererClass=&NiagaraSpriteRendererProperties, rendererIndex=1}
  ]
 }
 FallingStars {
  emitterUpdateScript [
   &EmitterState()
   &SpawnRate(Use Spawn Probability=false)
  ]
  particleSpawnScript [
   &InitializeParticle(
    Color=&MergeHelper_ColorSwitch(
     0=<rgba 1.5009 0.0208 2 1>
     1=<rgba 2 0.7381 0.0208 1>
     2=?
     3=?
     Index=&MakeIntFromFloat(Float=$User.VariantPicker)
     Num=::NewEnumerator0
    )
    Lifetime Max=1
    Lifetime Min=0.8
    Lifetime Mode=::NewEnumerator1
    Position Offset=<0 0 300>
    Sprite Rotation Mode=::NewEnumerator2
    Sprite Size=&Multiply_Vector2D_ByFloat(
     Float=&Multiply_Float(A=2.5, B=&RandomRangeFloat(Minimum=0.5, Maximum=1))
     Vector2D=<15 45>
    )
    Sprite Size Mode=::NewEnumerator4
    UsePositionOffset=true
    Write Material Random=true
   )
   &ShapeLocation(MaskSpawnGroup=false, Override Seed=false, Shape Primitive=::NewEnumerator4, UseOwnerScale=true)
   &AddVelocity(Velocity=&RandomRangeVector(Minimum=<0 0 -800>, Maximum=<0 0 -1200>))
   &DynamicMaterialParameters(
    Index 0 Param 1=140
    Module.Index0Param0Write=true
    Module.Index0Param1Write=true
    Module.Index0Param2Write=true
    Module.Index0Param3Write=true
   )
  ]
  particleUpdateScript [
   &ParticleState()
   &SolveForcesAndVelocity(Clamp Velocity=false, Limit Acceleration=false)
   &GenerateLocationEvent(Use Event Probability=false, UseEventDelay=false)
   &ScaleColor(
    Scale Alpha=&FNTimeOfDayScalar(Night Scale=0.8, Morning Scale=1, Evening Scale=1)
    ScaleA=true
    ScaleRGB=true
   ) {module=ScaleColor002}
  ]
  rendererClasses [&NiagaraSpriteRendererProperties]
  renderers [{rendererClass=&NiagaraSpriteRendererProperties}]
 }
 Minimal {
  emitterUpdateScript [
   &EmitterState()
   &SpawnRate(SpawnGroup=2, SpawnRate=30, Use Spawn Probability=false) {module=SpawnRate002}
   &SpawnRate(SpawnGroup=1, SpawnRate=&Lerp_Float(A=10, B=20, Alpha=$User.ParaGlideLeanAlpha), Use Spawn Probability=false) {module=SpawnRate001}
   &SpawnRate(SpawnRate=&Lerp_Float(A=6, B=8, Alpha=$User.ParaGlideLeanAlpha), Use Spawn Probability=true)
  ]
  particleSpawnScript [
   &InitializeParticle(
    Color Mode=::NewEnumerator0
    Lifetime=&MergeHelper_FloatSwitch(
     0=&RandomRangeFloat(Minimum=0.25, Maximum=1.25)
     1=&RandomRangeFloat(Minimum=0.6, Maximum=1.5)
     2=1
     3=?
     Index=&SpawnGroup()
     Num=::NewEnumerator1
    )
    Sprite Rotation Mode=::NewEnumerator2
    UsePositionOffset=false
    Write Material Random=true
   )
   /SetVariables(Particles.VisibilityTag=&SpawnGroup()) {isSetParameters=true}
   /SetVariables(
    Particles.StarsColor:LinearColor=&MergeHelper_ColorSwitch(
     0=<rgba 5.139 1 8 1>
     1=<rgba 8 3.537 1 1>
     2=?
     3=?
     Index=&MakeIntFromFloat(Float=$User.VariantPicker)
     Num=::NewEnumerator0
    )
    Particles.StarsSpriteSize:Vector2f=&Vector2DFromFloat(Value=&Multiply_Float(A=&RandomRangeFloat(Minimum=1, Maximum=3), B=1.5))
   ) {isSetParameters=true}
   /SetVariables(Particles.DustSize:Vector2f=&Vector2DFromFloat(Value=&Multiply_Float(
      A=&RandomRangeFloat(Minimum=45, Maximum=44)
      B=&Lerp_Float(A=1.9518, B=1.8, Alpha=$User.ParaGlideLeanAlpha)
     ))) {isSetParameters=true}
   /SetVariables(Particles.NumbersSize:Vector2f=&Vector2DFromFloat(Value=&RandomRangeFloat(Minimum=5, Maximum=10))) {isSetParameters=true}
   /SetVariables(Particles.ParticleSideMask:NiagaraInt32=&ModuloInt(A=&ReturnParticleIDAttributes(), B=2)) {isSetParameters=true}
  ]
  particleUpdateScript [
   &ParticleState()
   &SpriteRotationRate(Rotation Rate=&Multiply_Float(
     A=&UniformAOrBFloat(A=-1, Evaluation Type=::NewEnumerator0, Recalculate Random Each Loop=true, Override Seed=false)
     B=&RandomRangeFloat(Minimum=100, Maximum=250)
    )) {module=SpriteRotationRate001}
   &ScaleColor(
    Scale Alpha=&FNTimeOfDayScalar(Night Scale=0.8, Morning Scale=1, Evening Scale=1)
    ScaleA=true
    ScaleRGB=true
   ) {module=ScaleColor002}
   &SkeletalMeshLocation(
    Bone / Socket Index=&ModuloInt(A=&ReturnParticleIDAttributes(), B=2)
    Bone Sampling Mode=::NewEnumerator4
    MaskSpawnGroup=false
    Sampled Position Offset=&MergeHelper_VectorSwitch(
     0=&RandomRangeVector(Minimum=<-20 -20 -20>, Maximum=<20 20 20>)
     1=&RandomRangeVector(Minimum=<-20 -20 -20>, Maximum=<20 20 20>)
     2=<0 0 0>
     3=?
     Index=$Particles.VisibilityTag
     Num=::NewEnumerator1
    )
    Skeletal Mesh=$System.Skeletal Mesh Reader
   )
   /SetVariables(
    Particles.StarsColorUpdate:LinearColor=&MakeLinearColorFromVectorAndFloat(
     "Vector (RGB)"=&MakeVectorFromLinearColorRGB(Color=&MultiplyLinearColor(
       Linear Color=$Particles.StarsColor
       Scale Factor=&MakeVector4_FromLinearColor(LinearColor=&ColorFromCurve(ColorCurve=DataInterface() {propertyValues {
           AlphaCurve {keys [<at 0 1>,<at 1 0>]}
           BlueCurve {keys [<at 0 5>,<at 0.1051 1>,<at 0.2645 0.9531>,<at 0.517 0.7>]}
           GreenCurve {keys [<at 0 5>,<at 0.1051 1>,<at 0.2645 0.9531>,<at 0.517 0.7>]}
           RedCurve {keys [<at 0 5>,<at 0.1051 1>,<at 0.2645 0.9531>,<at 0.517 0.7>]}
          }}))
      ))
     "Float (Alpha)"=&FNTimeOfDayScalar(Night Scale=0.6, Morning Scale=0.8, Day Scale=0.8, Evening Scale=0.8)
    )
    Particles.StarsPosition:NiagaraPosition=&AddVectorToPosition(
     Position=$Particles.Position
     Vector=&MakeVector(
      X=&Lerp_Float(
       A=0
       B=&MergeHelper_FloatSwitch(
        0=&FloatFromCurve(
         FloatCurve=DataInterface() {propertyValues {Curve {defaultValue=1, keys [{tangentMode=RCTM_User, time=0, value=0}, {tangentMode=RCTM_User, time=1, value=1}]}}}
         Scale Curve=60
        )
        1=&FloatFromCurve(
         FloatCurve=DataInterface() {propertyValues {Curve {defaultValue=1, keys [{tangentMode=RCTM_User, time=0, value=0}, {tangentMode=RCTM_User, time=1, value=1}]}}}
         Scale Curve=-60
        )
        2=?
        3=?
        Index=$Particles.ParticleSideMask
        Num=::NewEnumerator0
       )
       Alpha=$User.ParaGlideLeanAlpha
      )
      Y=&Lerp_Float(
       A=0
       B=&FloatFromCurve(
        FloatCurve=DataInterface() {propertyValues {Curve {defaultValue=1, keys [{tangentMode=RCTM_User, time=0, value=0}, {tangentMode=RCTM_User, time=1, value=1}]}}}
        Scale Curve=&FN_FrontEndFloatModifier(Gameplay=0, Front End=-60)
       )
       Alpha=$User.ParaGlideLeanAlpha
      )
      Z=&FloatFromCurve(
       FloatCurve=DataInterface() {propertyValues {Curve {defaultValue=1, keys [{tangentMode=RCTM_User, time=0, value=0}, {tangentMode=RCTM_User, time=1, value=1}]}}}
       Scale Curve=&Multiply_Float(
        A=&RandomRangeFloat(Minimum=80, Maximum=150)
        B=&Lerp_Float(A=0.8, B=1.5, Alpha=$User.ParaGlideLeanAlpha)
       )
      )
     )
    )
    Particles.StarsSizeUpdate:Vector2f=&Multiply_Vector2D_ByFloat(
     Float=&FloatFromCurve(
      FloatCurve=DataInterface() {propertyValues {Curve {keys [
          {arriveTangent=14.26, interpMode=RCIM_Cubic, leaveTangent=14.26, tangentMode=RCTM_User, time=0, value=0}
          {arriveTangentWeight=0.0869, interpMode=RCIM_Cubic, leaveTangentWeight=0.1958, tangentMode=RCTM_User, time=0.2, value=1}
          <at 1 0>
         ]}}}
      Scale Curve=1.8
     )
     Vector2D=$Particles.StarsSpriteSize
    )
   ) {isSetParameters=true}
   /SetVariables(
    Particles.StarsGlowColor:LinearColor=&Multiply_LinearColorByFloat(
     Linear Color=&MultiplyLinearColor(Linear Color=$Particles.StarsColorUpdate, Scale Factor=&MakeVector4(X=1, Y=1, Z=1, W=0.1))
     Float=&FNTimeOfDayScalar(Night Scale=0.6, Morning Scale=0.8, Day Scale=0.8, Evening Scale=0.8)
    )
    Particles.StarsGlowSize:Vector2f=&Multiply_Vector2D_ByFloat(Float=3, Vector2D=$Particles.StarsSizeUpdate)
   ) {isSetParameters=true}
   /SetVariables(
    Particles.DustColor:LinearColor=&MakeLinearColorFromVectorAndFloat(
     "Vector (RGB)"=&MakeVectorFromLinearColorRGB(Color=&MergeHelper_ColorSwitch(
       0=<rgba 0.7795 0.2754 1 1>
       1=<rgba 1 0.6462 0.2754 1>
       2=?
       3=?
       Index=&MakeIntFromFloat(Float=$User.VariantPicker)
       Num=::NewEnumerator0
      ))
     "Float (Alpha)"=&FloatFromCurve(
      FloatCurve=DataInterface() {propertyValues {Curve {keys [<at 0 0 RCIM_Cubic>, {interpMode=RCIM_Cubic, tangentMode=RCTM_Break, time=0.2, value=1}, <at 1 0>]}}}
      Scale Curve=&Multiply_Float(A=&FN_FrontEndFloatModifier(Gameplay=1, Front End=0.3), B=0.5)
     )
    )
    Particles.DustPosition:NiagaraPosition=&AddVectorToPosition(
     Position=$Particles.Position
     Vector=&MakeVector(
      X=&Lerp_Float(
       A=0
       B=&MergeHelper_FloatSwitch(
        0=&FloatFromCurve(FloatCurve=DataInterface() {propertyValues {Curve {keys [<at 0 0>,<at 1 25>]}}}, Scale Curve=2)
        1=&FloatFromCurve(FloatCurve=DataInterface() {propertyValues {Curve {keys [<at 0 0>,<at 1 -25>]}}}, Scale Curve=2)
        2=?
        3=?
        Index=$Particles.ParticleSideMask
        Num=::NewEnumerator0
       )
       Alpha=$User.ParaGlideLeanAlpha
      )
      Y=&Lerp_Float(
       A=0
       B=&FloatFromCurve(
        FloatCurve=DataInterface() {propertyValues {Curve {keys [<at 0 0>,<at 1 -100>]}}}
        Scale Curve=&FN_FrontEndFloatModifier(Gameplay=0, Front End=0.8)
       )
       Alpha=$User.ParaGlideLeanAlpha
      )
      Z=&FloatFromCurve(
       FloatCurve=DataInterface() {propertyValues {Curve {keys [<at 0 0>,<at 1 80>]}}}
       Scale Curve=&FloatFromCurve(FloatCurve=DataInterface() {propertyValues {Curve {
           defaultValue=1
           keys [{interpMode=RCIM_Cubic, tangentMode=RCTM_User, time=0, value=0.5}, {interpMode=RCIM_Cubic, tangentMode=RCTM_User, time=1, value=1.8}]
          }}})
      )
     )
    )
    Particles.DustSizeUpdate:Vector2f=&Multiply_Vector2D_ByFloat(
     Float=&FloatFromCurve(FloatCurve=DataInterface() {propertyValues {Curve {keys [
          {arriveTangent=3.5003, interpMode=RCIM_Cubic, leaveTangent=3.5003, tangentMode=RCTM_User, time=0, value=0.5}
          {arriveTangent=0.5833, interpMode=RCIM_Cubic, leaveTangent=0.5833, time=0.18, value=0.9}
          <at 1.2 1.2 RCIM_Cubic>
         ]}}})
     Vector2D=$Particles.DustSize
    )
   ) {isSetParameters=true}
   /SetVariables(
    Particles.NumbersColor:LinearColor=&MakeLinearColorFromVectorAndFloat(
     "Vector (RGB)"=&MakeVectorFromLinearColorRGB(Color=&MergeHelper_ColorSwitch(
       0=&ColorFromCurve(ColorCurve=DataInterface() {propertyValues {
          AlphaCurve {keys [<at 0 1 RCIM_Cubic>,<at 1 0 RCIM_Cubic>]}
          BlueCurve {keys [<at 0 3 RCIM_Cubic>,<at 1 0.375 RCIM_Cubic>]}
          GreenCurve {keys [<at 0 0.4769 RCIM_Cubic>,<at 1 0.0005 RCIM_Cubic>]}
          RedCurve {keys [<at 0 2.9818 RCIM_Cubic>,<at 1 0.0656 RCIM_Cubic>]}
         }})
       1=&ColorFromCurve(ColorCurve=DataInterface() {propertyValues {
          AlphaCurve {keys [<at 0 1 RCIM_Cubic>,<at 1 0 RCIM_Cubic>]}
          BlueCurve {keys [<at 0 0 RCIM_Cubic>,<at 1 0.0005 RCIM_Cubic>]}
          GreenCurve {keys [<at 0 1.4528 RCIM_Cubic>,<at 1 0.0657 RCIM_Cubic>]}
          RedCurve {keys [<at 0 3 RCIM_Cubic>,<at 1 0.375 RCIM_Cubic>]}
         }})
       2=?
       3=?
       Index=&MakeIntFromFloat(Float=$User.VariantPicker)
       Num=::NewEnumerator0
      ))
     "Float (Alpha)"=&FloatFromCurve(
      FloatCurve=DataInterface() {propertyValues {Curve {keys [
          <at 0.102 0.0028 RCIM_Cubic>
          {arriveTangent=-0.0032, interpMode=RCIM_Cubic, leaveTangent=-0.0032, time=0.2344, value=1}
          <at 1 0 RCIM_Cubic>
         ]}}}
      Scale Curve=&Multiply_Float(A=1, B=&FNTimeOfDayScalar(Night Scale=0.6, Morning Scale=1, Evening Scale=1))
     )
    )
    Particles.NumbersPosition:NiagaraPosition=&AddVectorToPosition(
     Position=$Particles.Position
     Vector=&MakeVector(
      X=&Lerp_Float(
       A=0
       B=&MergeHelper_FloatSwitch(
        0=&FloatFromCurve(
         FloatCurve=DataInterface() {propertyValues {Curve {defaultValue=1, keys [{tangentMode=RCTM_User, time=0, value=0}, {tangentMode=RCTM_User, time=1, value=1}]}}}
         Scale Curve=10
        )
        1=&FloatFromCurve(
         FloatCurve=DataInterface() {propertyValues {Curve {defaultValue=1, keys [{tangentMode=RCTM_User, time=0, value=0}, {tangentMode=RCTM_User, time=1, value=1}]}}}
         Scale Curve=-10
        )
        2=?
        3=?
        Index=$Particles.ParticleSideMask
        Num=::NewEnumerator0
       )
       Alpha=$User.ParaGlideLeanAlpha
      )
      Y=&Lerp_Float(
       A=0
       B=&FloatFromCurve(
        FloatCurve=DataInterface() {propertyValues {Curve {defaultValue=1, keys [{tangentMode=RCTM_User, time=0, value=0}, {tangentMode=RCTM_User, time=1, value=1}]}}}
        Scale Curve=&FN_FrontEndFloatModifier(Gameplay=0, Front End=-60)
       )
       Alpha=$User.ParaGlideLeanAlpha
      )
      Z=&FloatFromCurve(
       FloatCurve=DataInterface() {propertyValues {Curve {defaultValue=1, keys [{tangentMode=RCTM_User, time=0, value=0}, {tangentMode=RCTM_User, time=1, value=1}]}}}
       Scale Curve=&Multiply_Float(
        A=&RandomRangeFloat(Minimum=20, Maximum=50)
        B=&Lerp_Float(A=0.8, B=1, Alpha=$User.ParaGlideLeanAlpha)
       )
      )
     )
    )
    Particles.NumbersSizeUpdate:Vector2f=&Multiply_Vector2D(
     A=&Vector2DFromFloat(Value=&FloatFromCurve(FloatCurve=DataInterface() {propertyValues {Curve {keys [
           <at 0.0113 0.4968 RCIM_Cubic>
           {arriveTangent=3.784, interpMode=RCIM_Cubic, leaveTangent=3.784, time=0.0978, value=1.5123}
           {arriveTangent=-0.5659, interpMode=RCIM_Cubic, leaveTangent=-0.5659, time=0.2277, value=1.3156}
           <at 1 1.0017>
          ]}}}))
     B=$Particles.NumbersSize
    )
   ) {isSetParameters=true}
   &SolveForcesAndVelocity(Clamp Velocity=false, Limit Acceleration=false)
  ]
  rendererClasses [&NiagaraSpriteRendererProperties]
  renderers [
   {rendererClass=&NiagaraSpriteRendererProperties}
   {rendererClass=&NiagaraSpriteRendererProperties, rendererIndex=1}
   {rendererClass=&NiagaraSpriteRendererProperties, rendererIndex=2}
   {rendererClass=&NiagaraSpriteRendererProperties, rendererIndex=3}
  ]
 }
 ShinnyDust {
  emitterUpdateScript [
   &EmitterState()
   &SpawnRate(SpawnRate=15, Use Spawn Probability=false)
  ]
  particleSpawnScript [
   &FN_SetSpawnGroup(Number of Spawn Groups=2)
   &InitializeParticle(
    Sprite Rotation Mode=::NewEnumerator2
    Sprite Size Mode=::NewEnumerator3
    Sprite UV Mode=::NewEnumerator3
    Uniform Sprite Size=&FN_PickFloatDependOnPlatform(
     TargetPlatformActive=&Multiply_Float(
      A=&RandomRangeFloat(Minimum=45, Maximum=44)
      B=&Lerp_Float(A=1.3, B=1.8, Alpha=$User.ParaGlideLeanAlpha)
     )
     TargetPlatformNoActive=&Multiply_Float(
      A=&RandomRangeFloat(Minimum=45, Maximum=44)
      B=&Lerp_Float(A=1.8, B=2, Alpha=$User.ParaGlideLeanAlpha)
     )
    )
    UsePositionOffset=false
    Write Material Random=true
   )
   &CameraOffset(Camera Offset Amount=&RandomRangeFloat(Minimum=0, Maximum=30))
  ]
  particleUpdateScript [
   &ParticleState()
   &ScaleColor(
    Scale Alpha=&FloatFromCurve(
     FloatCurve=DataInterface() {propertyValues {Curve {keys [<at 0 0 RCIM_Cubic>, {interpMode=RCIM_Cubic, tangentMode=RCTM_Break, time=0.2, value=1}, <at 1 0>]}}}
     Scale Curve=&Multiply_Float(
      A=&Multiply_Float(
       A=&FN_FrontEndFloatModifier(Gameplay=1, Front End=0.3)
       B=&self:SpawnGrp_0_1_Mask(
        Masked Float:NiagaraFloat=&RandomRangeFloat(Minimum=0.2, Maximum=0.4)
        Passthrough Float:NiagaraFloat=1
       )
      )
      B=&FN_PickFloatDependOnPlatform(TargetPlatformActive=1, TargetPlatformNoActive=0.6)
     )
    )
    Scale RGB=&MakeVectorFromLinearColorRGB(Color=&MergeHelper_ColorSwitch(
      0=<rgba 0.7795 0.2754 1 1>
      1=<rgba 1 0.6462 0.2754 1>
      2=?
      3=?
      Index=&MakeIntFromFloat(Float=$User.VariantPicker)
      Num=::NewEnumerator0
     ))
    ScaleA=true
    ScaleRGB=true
   )
   &ScaleColor(
    Scale Alpha=&FNTimeOfDayScalar(Night Scale=0.8, Morning Scale=1, Evening Scale=1)
    ScaleA=true
    ScaleRGB=true
   ) {module=ScaleColor002}
   &ScaleSpriteSize(Uniform Curve Sprite Scale=DataInterface() {propertyValues {Curve {keys [
        {arriveTangent=3.5003, interpMode=RCIM_Cubic, leaveTangent=3.5003, tangentMode=RCTM_User, time=0, value=0.5}
        {arriveTangent=0.5833, interpMode=RCIM_Cubic, leaveTangent=0.5833, time=0.18, value=0.9}
        <at 1.2 1.2 RCIM_Cubic>
       ]}}})
   &SkeletalMeshLocation(
    Bone / Socket Index=$Particles.SpawnGroup
    Bone Sampling Mode=::NewEnumerator4
    MaskSpawnGroup=false
    Orientation Sampling=::NewEnumerator1
    Sampled Position Offset=&MakeVector(
     X=&Lerp_Float(
      A=0
      B=&MaskFloatBySpawnGroup(
       Masked Float=&FloatFromCurve(FloatCurve=DataInterface() {propertyValues {Curve {keys [<at 0 0>,<at 1 25>]}}})
       Passthrough Float=&FloatFromCurve(FloatCurve=DataInterface() {propertyValues {Curve {keys [<at 0 0>,<at 1 -25>]}}})
      )
      Alpha=$User.ParaGlideLeanAlpha
     )
     Y=&Lerp_Float(
      A=0
      B=&FloatFromCurve(
       FloatCurve=DataInterface() {propertyValues {Curve {keys [<at 0 0>,<at 1 -100>]}}}
       Scale Curve=&FN_FrontEndFloatModifier(Gameplay=0, Front End=0.8)
      )
      Alpha=$User.ParaGlideLeanAlpha
     )
     Z=&FloatFromCurve(
      FloatCurve=DataInterface() {propertyValues {Curve {keys [<at 0 0>,<at 1 80>]}}}
      Scale Curve=&self:SpawnGrp_0_1_Mask_01(Masked Float:NiagaraFloat=0.5, Passthrough Float:NiagaraFloat=1.8)
     )
    )
    Sampled Position Offset Coordinate Space:ENiagaraSkelMeshTransforms=::NewEnumerator2
    Skeletal Mesh=$System.Skeletal Mesh Reader
    Use Sampled Mesh Attribute as Ribbon ID=false
   )
  ]
  rendererClasses [&NiagaraSpriteRendererProperties]
  renderers [{rendererClass=&NiagaraSpriteRendererProperties}]
 }
 StarsDustTrail {
  emitterUpdateScript [
   &EmitterState()
   &SpawnParticlesFromOtherEmitter(
    Attribute Reader=DataInterface() {propertyValues {EmitterBinding {bindingMode=Other, emitterName=FallingStars}}}
    Spawn Rate=25
    Use Maximum Age=false
    Use Minimum Age=false
   )
  ]
  particleSpawnScript [
   &FN_SetSpawnGroup(Number of Spawn Groups=2)
   /SetVariables(Particles.VisibilityTag=&SpawnGroup()) {isSetParameters=true}
   &InitializeParticle(
    Color=&MergeHelper_ColorSwitch(
     0=<rgba 5.139 1 8 1>
     1=<rgba 8 3.537 1 1>
     2=?
     3=?
     Index=&MakeIntFromFloat(Float=$User.VariantPicker)
     Num=::NewEnumerator0
    )
    Lifetime=&MergeHelper_FloatSwitch(
     0=&RandomRangeFloat(Minimum=0.2, Maximum=0.3)
     1=&RandomRangeFloat(Minimum=0.5, Maximum=0.6)
     2=?
     3=?
     Index=$Particles.VisibilityTag
     Num=::NewEnumerator0
    )
    Position Offset=<0 0 35>
    Sprite Size Mode=::NewEnumerator3
    Uniform Sprite Size=&MergeHelper_FloatSwitch(
     0=&Multiply_Float(A=&RandomRangeFloat(Minimum=2, Maximum=6), B=1)
     1=&Multiply_Float(A=&RandomRangeFloat(Minimum=40, Maximum=50), B=1.4)
     2=?
     3=?
     Index=$Particles.VisibilityTag
     Num=::NewEnumerator0
    )
    UsePositionOffset=true
    Write Material Random=true
   )
   &SampleParticlesFromOtherEmitter(
    Applied Velocity Scale:NiagaraFloat=&MergeHelper_FloatSwitch(0=0, 1=0.5, 2=?, 3=?, Index=$Particles.VisibilityTag, Num=::NewEnumerator0)
    Transform Data Option=1
    Velocity Sampling=::NewEnumerator0
   )
   &ShapeLocation(
    Disc Coverage=0.4643
    MaskSpawnGroup=false
    Override Seed=false
    Ring Radius=&MergeHelper_FloatSwitch(0=15, 1=2, 2=?, 3=?, Index=$Particles.VisibilityTag, Num=::NewEnumerator0)
    Rotation Mode=::NewEnumerator2
    Shape Primitive=::NewEnumerator4
    UseOwnerScale=true
   )
  ]
  particleUpdateScript [
   &ParticleState()
   &SpriteRotationRate(Rotation Rate=&Multiply_Float(
     A=&UniformAOrBFloat(A=-1, Evaluation Type=::NewEnumerator0, Recalculate Random Each Loop=true, Override Seed=false)
     B=&RandomRangeFloat(Minimum=100, Maximum=250)
    ))
   &ScaleColor(
    Linear Color Curve=DataInterface() {propertyValues {
      AlphaCurve {keys [<at 0 1>,<at 1 0>]}
      BlueCurve {keys [<at 0 5>,<at 0.1051 1>,<at 0.2645 0.9531>,<at 0.517 0.7>]}
      GreenCurve {keys [<at 0 5>,<at 0.1051 1>,<at 0.2645 0.9531>,<at 0.517 0.7>]}
      RedCurve {keys [<at 0 5>,<at 0.1051 1>,<at 0.2645 0.9531>,<at 0.517 0.7>]}
     }}
    Scale Mode=::NewEnumerator2
   )
   &ScaleColor(
    Scale Alpha=&FloatFromCurve(FloatCurve=DataInterface() {propertyValues {Curve {
        defaultValue=1
        keys [{interpMode=RCIM_Cubic, tangentMode=RCTM_User, time=0, value=1}, {interpMode=RCIM_Cubic, tangentMode=RCTM_User, time=1, value=0}]
       }}})
    ScaleA=true
    ScaleRGB=true
   ) {module=ScaleColor001}
   &ScaleSpriteSize(Uniform Curve Sprite Scale=DataInterface() {propertyValues {Curve {
       defaultValue=0
       keys [
        {interpMode=RCIM_Cubic, tangentMode=RCTM_User, time=0, value=0}
        {interpMode=RCIM_Cubic, tangentMode=RCTM_User, time=0.5, value=1}
        {tangentMode=RCTM_User, time=1, value=0}
       ]
      }}})
   /SetVariables(
    Particles.StarsGlowSize:Vector2f=&Multiply_Vector2D_ByFloat(Float=3, Vector2D=$Particles.SpriteSize)
    Particles.StarsGlowlowColor:LinearColor=&MultiplyLinearColor(Linear Color=$Particles.Color, Scale Factor=&MakeVector4(X=1, Y=1, Z=1, W=0.1))
   ) {isSetParameters=true}
   /SetVariables(
    Particles.DustColor:LinearColor=&MakeLinearColorFromVectorAndFloat(
     "Vector (RGB)"=&MakeVectorFromLinearColorRGB(Color=&MergeHelper_ColorSwitch(
       0=<rgba 0.7795 0.2754 1 1>
       1=<rgba 1 0.6462 0.2754 1>
       2=?
       3=?
       Index=&MakeIntFromFloat(Float=$User.VariantPicker)
       Num=::NewEnumerator0
      ))
     "Float (Alpha)"=&FloatFromCurve(
      FloatCurve=DataInterface() {propertyValues {Curve {keys [<at 0 0 RCIM_Cubic>, {interpMode=RCIM_Cubic, tangentMode=RCTM_Break, time=0.2, value=1}, <at 1 0>]}}}
      Scale Curve=&FN_FrontEndFloatModifier(Gameplay=1, Front End=0.3)
     )
    )
    Particles.DustSize:Vector2f=&Multiply_Vector2D(
     A=$Particles.SpriteSize
     B=&Vector2DFromFloat(Value=&FloatFromCurve(FloatCurve=DataInterface() {propertyValues {Curve {keys [
           {arriveTangent=3.5003, interpMode=RCIM_Cubic, leaveTangent=3.5003, tangentMode=RCTM_User, time=0, value=0.5}
           {arriveTangent=0.5833, interpMode=RCIM_Cubic, leaveTangent=0.5833, time=0.18, value=0.9}
           <at 1.2 1.2 RCIM_Cubic>
          ]}}}))
    )
   ) {isSetParameters=true}
   &Drag(Drag=2, Use Linear Drag=true, Use Rotational Drag=true)
   &CurlNoiseForce(Noise Strength=200, Pan Noise=false)
   &SolveForcesAndVelocity(Clamp Velocity=false, Limit Acceleration=false)
  ]
  rendererClasses [&NiagaraSpriteRendererProperties]
  renderers [
   {rendererClass=&NiagaraSpriteRendererProperties}
   {rendererClass=&NiagaraSpriteRendererProperties, rendererIndex=1}
   {rendererClass=&NiagaraSpriteRendererProperties, rendererIndex=2}
  ]
 }
 StarsTrail {
  emitterUpdateScript [
   &EmitterState()
   &SpawnParticlesFromOtherEmitter(
    Attribute Reader=DataInterface() {propertyValues {EmitterBinding {bindingMode=Other, emitterName=FallingStars}}}
    Spawn Rate=50
    Use Maximum Age=false
    Use Minimum Age=false
   )
  ]
  particleSpawnScript [
   &InitializeParticle(
    Lifetime=0.2
    Ribbon Width=10
    Ribbon Width Mode=::NewEnumerator1
    Sprite Rotation Mode=::NewEnumerator2
    UsePositionOffset=false
    Write Material Random=true
   )
   &SampleParticlesFromOtherEmitter(
    Age Sampling:ENiagara_DirectReadSamplingMode=::NewEnumerator0
    Lifetime Application:ENiagara_LifetimeInheritanceOptions=::NewEnumerator1
    Lifetime Sampling:ENiagara_DirectReadSamplingMode=::NewEnumerator0
    Minimum Sampled Particle Age=0.01
    Module.MaxSampledAgeCutoff=false
    Module.MinSampledAgeCutoff=true
    Reject Particle By Sampled Age=true
   )
   /SetVariables(Particles.NormalizedGuideParticleAge:NiagaraFloat=&Divide_Float(A=$Output.SampleParticlesFromOtherEmitter.SampledAge, B=$Output.SampleParticlesFromOtherEmitter.SampledLifetime)) {isSetParameters=true}
  ]
  particleUpdateScript [
   &ParticleState()
   &ScaleRibbonWidth(Ribbon Width Scale=&FloatFromCurve(FloatCurve=DataInterface() {propertyValues {Curve {keys [<at 0.0105 0.5324>,<at 0.1077 0.9765>,<at 1 0.15>]}}}, Scale Curve=0.6))
   &ScaleColor(
    Scale Alpha=&FNTimeOfDayScalar(Night Scale=0.8, Morning Scale=1, Evening Scale=1)
    ScaleA=true
    ScaleRGB=true
   ) {module=ScaleColor002}
   &ScaleColor(
    Scale RGB=&MakeVectorFromLinearColorRGB(Color=&MergeHelper_ColorSwitch(
      0=&ColorFromCurve(ColorCurve=DataInterface() {propertyValues {
         AlphaCurve {keys [<at 0 0>,<at 0.0889 1>,<at 0.5313 1>,<at 1 0>]}
         BlueCurve {keys [<at 0 3>,<at 0.2808 1.6985>,<at 1 0.4688>]}
         GreenCurve {keys [<at 0 0>,<at 0.2808 0>,<at 1 0.0451>]}
         RedCurve {keys [<at 0 0.9913>,<at 0.2808 0.8271>,<at 1 0.2072>]}
        }})
      1=&ColorFromCurve(ColorCurve=DataInterface() {propertyValues {
         AlphaCurve {keys [<at 0 0>,<at 0.0889 1>,<at 0.5313 1>,<at 1 0>]}
         BlueCurve {keys [<at 0 0>,<at 0.2808 0>,<at 1 0.0451>]}
         GreenCurve {keys [<at 0 2.341>,<at 0.2808 0.5426>,<at 1 0.1064>]}
         RedCurve {keys [<at 0 3>,<at 0.2808 1.6985>,<at 1 0.4688>]}
        }})
      2=?
      3=?
      Index=&MakeIntFromFloat(Float=$User.VariantPicker)
      Num=::NewEnumerator0
     ))
    ScaleA=true
    ScaleRGB=true
   ) {module=ScaleColor001}
   &SolveForcesAndVelocity(Clamp Velocity=false, Limit Acceleration=false)
  ]
  rendererClasses [&NiagaraRibbonRendererProperties]
  renderers [{rendererClass=&NiagaraRibbonRendererProperties}]
 }
 Trail {
  emitterUpdateScript [
   &EmitterState()
   &SpawnRate(SpawnRate=2, Use Spawn Probability=false)
  ]
  particleSpawnScript [
   &FN_SetSpawnGroup(Number of Spawn Groups=2)
   &PartitionParticles()
   &AddVelocity(
    Scale Added Velocity=&Lerp_Vector(A=<1 0 1>, B=<1 1 2>, Alpha=$System.FN_Integrator001.Value)
    Velocity=<0 -600 500>
   )
   &InitializeParticle(
    Color=<rgba 1 1 1 0.25>
    Lifetime Max=3
    Lifetime Min=2
    Lifetime Mode=::NewEnumerator1
    Mass Mode=::NewEnumerator1
    Mesh Scale Mode=::NewEnumerator3
    Mesh Uniform Scale=1.2
    Sprite Size Mode=::NewEnumerator4
    UsePositionOffset=false
    Write Material Random=true
   ) {module=InitializeParticle001}
   &InitialMeshOrientation(
    Mesh Axis To Orient=<0 1 0>
    Mesh Orientation Mode=::NewEnumerator2
    Orientation Vector=<0 1 0>
    Rotation=&Lerp_Vector(
     A=<0 0 0>
     B=&FN_FrontEndVector3Modifier(
      Gameplay=&MaskVectorBySpawnGroup(Masked Vector=<0 0.02 0>, Passthrough Vector=<0 -0.02 0>)
      Front End=&MaskVectorBySpawnGroup(Masked Vector=<0.05 0.02 0>, Passthrough Vector=<0.05 -0.02 0>)
     )
     Alpha=$User.ParaGlideLeanAlpha
    )
    Use Rotation Vector=true
   )
  ]
  particleUpdateScript [
   &ParticleState()
   &UpdateMeshOrientation(
    Rotation Rate=&SelectFloatFromArray(
     Direct Array Index=$Particles.Initial.PartitionParticles.Partition
     Float Selection Array=DataInterface() {propertyValues {FloatData [-1,1]}}
    )
    Rotation Vector=<0 0 1>
   )
   &SkeletalMeshLocation(
    Bone / Socket Index=$Particles.SpawnGroup
    Bone Sampling Mode=::NewEnumerator4
    Inherit Velocity=false
    Interpolate Mesh Sampling=false
    MaskSpawnGroup=false
    Sampled Position Offset=<0 0 -30>
    Skeletal Mesh=$System.Skeletal Mesh Reader
   )
   &ScaleMeshSize(Scale Factor=&VectorFromCurve(
     VectorCurve=DataInterface() {propertyValues {
       XCurve {keys [<at 0 1>,<at 1 1>]}
       YCurve {keys [<at 0 1>,<at 1 1>]}
       ZCurve {keys [{interpMode=RCIM_Cubic, leaveTangent=2.7698, tangentMode=RCTM_Break, time=0, value=1}, <at 0.2 1.375>, <at 1 2>]}
      }}
     Scale Curve=&MakeVector(X=1, Y=1, Z=&Lerp_Float(A=1, B=1.2, Alpha=$User.ParaGlideLeanAlpha))
    ))
   &SolveForcesAndVelocity(Clamp Velocity=false, Limit Acceleration=false)
   &ScaleColor(
    Scale Alpha=&Multiply_Float(A=1, B=$System.FadeIn)
    Scale RGB=&MakeVectorFromLinearColorRGB(Color=&MergeHelper_ColorSwitch(
      0=&ColorFromCurve(ColorCurve=DataInterface() {propertyValues {
         AlphaCurve {keys [<at 0 0 RCIM_Cubic>,<at 1 1 RCIM_Cubic>]}
         BlueCurve {keys [<at 0 10 RCIM_Cubic>,<at 0.3612 2.007>,<at 0.554 1.3695>,<at 0.7735 0.7482>,<at 1 0.4688 RCIM_Cubic>]}
         GreenCurve {keys [<at 0 1.25 RCIM_Cubic>,<at 0.3612 0.0378>,<at 0.554 0.0408>,<at 0.7735 0.0437>,<at 1 0.0598 RCIM_Cubic>]}
         RedCurve {keys [<at 0 9.1629 RCIM_Cubic>,<at 0.3612 0.8939>,<at 0.554 1.0344>,<at 0.7735 0.4603>,<at 1 0.0451 RCIM_Cubic>]}
        }})
      1=&ColorFromCurve(ColorCurve=DataInterface() {propertyValues {
         AlphaCurve {keys [<at 0 0 RCIM_Cubic>,<at 1 1 RCIM_Cubic>]}
         BlueCurve {keys [<at 0 0 RCIM_Cubic>,<at 0.3612 0.0378>,<at 0.554 0.0408>,<at 0.7735 0.0437>,<at 1 0.0451 RCIM_Cubic>]}
         GreenCurve {keys [<at 0 3.7702 RCIM_Cubic>,<at 0.3612 0.6977>,<at 0.554 0.4526>,<at 0.7735 0.2138>,<at 1 0.1064 RCIM_Cubic>]}
         RedCurve {keys [<at 0 10 RCIM_Cubic>,<at 0.3612 2.007>,<at 0.554 1.3695>,<at 0.7735 0.7482>,<at 1 0.4688 RCIM_Cubic>]}
        }})
      2=?
      3=?
      Index=&MakeIntFromFloat(Float=$User.VariantPicker)
      Num=::NewEnumerator0
     ))
    ScaleA=true
    ScaleRGB=true
   )
   &ScaleColor(
    Scale Alpha=&FNTimeOfDayScalar(Night Scale=0.8, Morning Scale=1, Evening Scale=1)
    ScaleA=true
    ScaleRGB=true
   ) {module=ScaleColor002}
  ]
  rendererClasses [&NiagaraMeshRendererProperties]
  renderers [{rendererClass=&NiagaraMeshRendererProperties, rendererIndex=0}]
 }
}
summary {
 systemName=NS_Contrail_IndoorLaceScore
 emitters [
  {emitterName=Trail, rendererClasses [&NiagaraMeshRendererProperties]}
  {emitterName=ShinnyDust, rendererClasses [&NiagaraSpriteRendererProperties]}
  {emitterName=FallingStars, rendererClasses [&NiagaraSpriteRendererProperties]}
  {emitterName=BigRibbon, rendererClasses [&NiagaraRibbonRendererProperties]}
  {emitterName=ClockMain, rendererClasses [&NiagaraSpriteRendererProperties, &NiagaraLightRendererProperties]}
  {emitterName=ClockSecondary, rendererClasses [&NiagaraSpriteRendererProperties]}
  {emitterName=StarsTrail, rendererClasses [&NiagaraRibbonRendererProperties]}
  {emitterName=Minimal, rendererClasses [&NiagaraSpriteRendererProperties]}
  {emitterName=StarsDustTrail, rendererClasses [&NiagaraSpriteRendererProperties]}
 ]
 userVariables [
  User.bIsFrontEnd {type=NiagaraBool}
  User.bIsFrontEndPreview {type=NiagaraBool}
  User.DynamicFloat {type=NiagaraFloat, defaultValue=0, description=""}
  User.ParaGlideLeanAlpha {type=NiagaraFloat, defaultValue=0, description=""}
  User.RibbonIDOffset {type=NiagaraInt32}
  User.SkeletalMesh {
   type=NiagaraDataInterfaceSkeletalMesh
   defaultValue {
    struct=NiagaraExt_VariableValue_DataInterface
    value {dataInterface=&self:NiagaraDataInterfaceSkeletalMesh_0}
   }
  }
  User.VariantPicker {type=NiagaraFloat, defaultValue=0, description=""}
 ]
}
